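# Physics asset for the Phoenix bird skeleton in the dealership shared library's set assets, giving its bones simple collision volumes.
# Each body is a single capsule, covering head, jaw, hair fins, wing feathers, tail fin chains, legs, toes and claws.
# Constraints follow the bone hierarchy from the pelvis, with wing feathers hung off wrists and elbows and fin chains off head and tail.
# All joints share the same uniform swing and twist limit values.
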PhysicsAsset /DealershipSharedLibrary/Assets/Sets/Phoenix/PA_DLR_Phoenix
bodies {
 claw_left_jnt [{shapeType=Capsule, length=45.8101, radius=16.8424, center=<7.5114 -2.5647 -2.5516>, rotation=<rot 2.1519 -90 0.7852>}]
 claw_right_jnt [{shapeType=Capsule, length=45.7931, radius=17.9614, center=<6.039 -2.5722 2.3788>, rotation=<rot -2.3504 -90 0.0384>}]
 elbow_left_jnt [{shapeType=Capsule, length=139.9028, radius=63.3412, center=<64.5281 1.9537 39.1926>, rotation=<rot 0 -92.1348 -77.8315>}]
 elbow_right_jnt [{shapeType=Capsule, length=144.4042, radius=64.9589, center=<66.2131 -5.5414 -22.2632>, rotation=<rot 0 88.0764 -76.7078>}]
 foot_left_jnt [{shapeType=Capsule, length=51.5315, radius=43.6764, center=<17.182 -5.2187 -2.0309>, rotation=<rot 0 -66.6816 -87.6559>}]
 foot_right_jnt [{shapeType=Capsule, length=51.5637, radius=43.6807, center=<17.6184 -6.8527 2.1318>, rotation=<rot 0 114.2926 -87.8047>}]
 hair_fin_center_a_01_jnt [{shapeType=Capsule, length=33.9994, radius=8.8391, center=<31.8754 0.6662 0.0001>, rotation=<rot 0 91.2639 -90>}]
 hair_fin_center_a_02_jnt [{shapeType=Capsule, length=47.5045, radius=15.8759, center=<54.1065 7.1205 0.0001>, rotation=<rot 0 -72.2546 -90>}]
 hair_fin_center_b_01_jnt [{shapeType=Capsule, length=94.1322, radius=9.8034, center=<82.0497 -21.7118 0.0001>, rotation=<rot 0 71.5003 -90>}]
 hair_fin_center_c_01_jnt [{shapeType=Capsule, length=54.3667, radius=7.4974, center=<55.0562 -3.2043 0.0001>, rotation=<rot 0 83.482 -90>}]
 hair_fin_center_d_01_jnt [{shapeType=Capsule, length=37.7939, radius=5.5682, center=<-25.8904 27.4388 0.0001>, rotation=<rot 0 -139.7934 -90>}]
 hair_fin_left_a_01_jnt [{shapeType=Capsule, length=40.508, radius=7.7291, center=<34.3404 -0.4228 -1.3729>, rotation=<rot 0 -87.7418 -87.059>}]
 hair_fin_left_a_02_jnt [{shapeType=Capsule, length=46.1882, radius=8.1147, center=<40.8216 -8.5885 -4.0269>, rotation=<rot 0 -90.7382 -86.2434>}]
 hair_fin_left_b_01_jnt2 [{shapeType=Capsule, length=51.523, radius=12.8355, center=<37.3058 6.5164 0.7305>, rotation=<rot 0 104.849 -89.7174>}]
 hair_fin_right_a_01_jnt [{shapeType=Capsule, length=87.7976, radius=9.6366, center=<74.5803 -4.2986 3.5662>, rotation=<rot 0 86.2656 -88.0517>}]
 hair_fin_right_b_01_jnt2 [{shapeType=Capsule, length=48.612, radius=12.8328, center=<40.7677 6.6107 -0.7138>, rotation=<rot 0 107.9857 -89.7192>}]
 head_JNT [{shapeType=Capsule, length=100.6661, radius=62.1539, center=<32.7189 -10.2355 0.0205>, rotation=<rot 0 -88.7703 -89.3288>}]
 hide_toe_left_jnt [{shapeType=Capsule, length=22.9535, radius=10.7752, center=<14.673 0.9183 -1.2436>, rotation=<rot 0 94.6356 -89.4939>}]
 hide_toe_right_jnt [{shapeType=Capsule, length=23.0324, radius=10.7654, center=<14.5831 0.6176 1.3344>, rotation=<rot 0 -82.8032 -89.8365>}]
 jaw_JNT [{shapeType=Capsule, length=35.761, radius=35.5796, center=<-4.3446 8.7335 2.6907>, rotation=<rot 4.0124 -90 -130.4395>}]
 neck_JNT [{shapeType=Capsule, length=47.8896, radius=42.953, center=<31.731 -12.7183 1.9803>, rotation=<rot 78.6765 54.8482 145.3766>}]
 pelvis_JNT [{shapeType=Capsule, length=102.6763, radius=84.2307, center=<-39.0995 -7.1405 1.1031>, rotation=<rot 88.4213 75.4909 165.4961>}]
 sfine_jnt [{shapeType=Capsule, length=108.4952, radius=104.634, center=<-1.5134 -15.1252 -0.0444>, rotation=<rot 0.1023 -116.9419 -90>}]
 shoudler_left_jnt [{shapeType=Capsule, length=19.4722, radius=17.3272, center=<25.2779 4.7916 -15.3118>, rotation=<rot 0 139.3449 -66.1726>}]
 shoudler_right_jnt [{shapeType=Capsule, length=19.0956, radius=11.727, center=<21.4273 12.6327 16.7293>, rotation=<rot 0 -62.9384 -53.4331>}]
 tail_JNT [{shapeType=Capsule, length=59.724, radius=59.3991, center=<20.608 -10.6142 0.6438>, rotation=<rot 0 -4.7734 -159.1782>}]
 tail_fin_a_01_jnt [{shapeType=Capsule, length=43.3738, radius=29.514, center=<45.3759 2.4583 -1.0704>, rotation=<rot 0 -87.515 -86.1974>}]
 tail_fin_a_02_jnt [{shapeType=Capsule, length=41.6702, radius=32.9366, center=<39.6069 5.1682 -2.8333>, rotation=<rot 0 -86.9566 -79.2781>}]
 tail_fin_a_03_jnt [{shapeType=Capsule, length=48.5591, radius=27.7143, center=<40.7159 5.028 0.2309>, rotation=<rot 0 -90.9087 -86.1326>}]
 tail_fin_b_01_jnt [{shapeType=Capsule, length=56.765, radius=22.2979, center=<50.6242 15.3039 -1.3377>, rotation=<rot 0 -83.4137 -89.6875>}]
 tail_fin_b_02_jnt [{shapeType=Capsule, length=105.0599, radius=39.0969, center=<99.7913 38.1359 5.1078>, rotation=<rot 0 109.2333 -86.5275>}]
 tail_fin_b_03_jnt [{shapeType=Capsule, length=48.2641, radius=27.286, center=<51.8872 5.2078 3.8407>, rotation=<rot 0 -89.8865 -81.921>}]
 tail_fin_c_01_jnt [{shapeType=Capsule, length=51.7438, radius=15.1947, center=<59.5784 -1.8256 -3.1376>, rotation=<rot 0 -86.1881 -88.2027>}]
 tail_fin_c_02_jnt [{shapeType=Capsule, length=62.2313, radius=22.3415, center=<70.1437 -2.0335 2.5521>, rotation=<rot 0 88.4334 -87.7893>}]
 tail_fin_c_03_jnt [{shapeType=Capsule, length=55.057, radius=33.8232, center=<92.3079 3.3332 5.5795>, rotation=<rot 0 93.8727 -86.9627>}]
 tail_fin_d_01_jnt [{shapeType=Capsule, length=60.7423, radius=18.0941, center=<59.6026 -10.1919 -2.2906>, rotation=<rot 0 -77.5608 -87.8133>}]
 tail_fin_d_02_jnt [{shapeType=Capsule, length=49.8163, radius=21.3438, center=<60.6349 -0.4352 -1.3139>, rotation=<rot 0 94.1505 -88.2043>}]
 tail_fin_d_03_jnt [{shapeType=Capsule, length=66.3302, radius=33.7751, center=<92.3932 4.5181 -0.1797>, rotation=<rot 0 90.8466 -89.6987>}]
 tail_fin_e_01_jnt [{shapeType=Capsule, length=70.5725, radius=24.9754, center=<66.3151 13.8567 -8.6401>, rotation=<rot 0 86.8232 -84.4392>}]
 tail_fin_e_02_jnt [{shapeType=Capsule, length=32.4494, radius=21.9477, center=<75.2766 12.6398 -0.1859>, rotation=<rot 0 94.6064 -63.1448>}]
 tail_fin_f_01_jnt [{shapeType=Capsule, length=183.0838, radius=31.5715, center=<167.0115 36.2485 24.7542>, rotation=<rot 0 110.3303 -78.867>}]
 tail_fin_f_02_jnt [{shapeType=Capsule, length=39.1213, radius=20.0022, center=<61.306 2.1671 -4.9494>, rotation=<rot 0 -82.7594 -75.2849>}]
 tail_fin_g_01_jnt [{shapeType=Capsule, length=78.6307, radius=20.8247, center=<77.8752 -5.0227 6.7843>, rotation=<rot 0 99.0254 -83.9289>}]
 tail_fin_g_02_jnt [{shapeType=Capsule, length=83.5121, radius=18.3308, center=<86.911 3.6337 0.9128>, rotation=<rot 0 -88.3429 -88.6663>}]
 tail_fin_g_03_jnt [{shapeType=Capsule, length=59.1024, radius=18.4584, center=<84.8064 14.3114 -8.8327>, rotation=<rot 0 -84.7686 -83.2987>}]
 tail_fin_h_01_jnt [{shapeType=Capsule, length=57.1627, radius=15.2053, center=<64.9729 -0.0264 0.1766>, rotation=<rot 0 93.2499 -89.0712>}]
 tail_fin_h_02_jnt [{shapeType=Capsule, length=64.5269, radius=18.0041, center=<77.7807 -5.6134 -4.3912>, rotation=<rot 0 87.4708 -87.5203>}]
 tail_fin_h_03_jnt [{shapeType=Capsule, length=37.7257, radius=14.4995, center=<73.2886 -6.4534 3.127>, rotation=<rot 0 -91.1247 -86.3409>}]
 tail_fin_i_01_jnt [{shapeType=Capsule, length=63.0797, radius=34.0604, center=<60.274 8.5477 6.5975>, rotation=<rot 0 107.3116 -74.2744>}]
 tail_fin_i_02_jnt [{shapeType=Capsule, length=101.4367, radius=48.366, center=<95.8669 43.9962 25.8596>, rotation=<rot 0 116.091 -74.0218>}]
 tail_fin_i_03_jnt [{shapeType=Capsule, length=42.1354, radius=20.5093, center=<71.2184 19.4302 -8.1288>, rotation=<rot 0 -72.5879 -84.2835>}]
 tarsus_left_jnt [{shapeType=Capsule, length=28.7254, radius=15.4574, center=<25.7667 1.6778 0.8827>, rotation=<rot 0 -98.1083 -89.6908>}]
 tarsus_right_jnt [{shapeType=Capsule, length=28.82, radius=15.4571, center=<25.6687 1.6826 -0.871>, rotation=<rot 0 81.9329 -89.7156>}]
 thigh_left_jnt [{shapeType=Capsule, length=71.8912, radius=42.0264, center=<73.6235 3.348 8.9051>, rotation=<rot 0 -84.8031 -88.8313>}]
 thigh_right_jnt [{shapeType=Capsule, length=71.8941, radius=42.0268, center=<73.6207 3.3451 -8.9105>, rotation=<rot 0 95.1896 -88.8457>}]
 toe_left_jnt [{shapeType=Capsule, length=45.5942, radius=13.4354, center=<8.4704 1.5618 -2.297>, rotation=<rot 1.9708 -90 0.969>}]
 toe_right_jnt [{shapeType=Capsule, length=44.9377, radius=12.8222, center=<9.1626 1.3561 2.6597>, rotation=<rot -2.3674 -90 0.8695>}]
 wing_a_left_01_jnt [{shapeType=Capsule, length=82.182, radius=41.7171, center=<85.1103 -5.5585 1.2768>, rotation=<rot 0 89.9846 -79.6245>}]
 wing_a_left_02_jnt [{shapeType=Capsule, length=94.1922, radius=33.1677, center=<29.5644 -9.2878 18.0502>, rotation=<rot 0 -93.5533 -56.3751>}]
 wing_a_right_01_jnt [{shapeType=Capsule, length=84.3981, radius=50.7565, center=<85.7232 -5.3344 -8.5451>, rotation=<rot 0 -89.7814 -78.2012>}]
 wing_a_right_02_jnt [{shapeType=Capsule, length=110.6689, radius=56.4226, center=<48.5953 -6.5089 -68.4628>, rotation=<rot -1.591 -90 28.7069>}]
 wing_b_left_01_jnt [{shapeType=Capsule, length=20.0764, radius=10.2488, center=<100.6327 0.2003 37.2175>, rotation=<rot 0 -93.6298 -68.1837>}]
 wing_b_left_02_jnt [{shapeType=Capsule, length=106.4427, radius=46.0218, center=<86.0162 -4.0236 -30.0533>, rotation=<rot 0 -90.6433 -54.6506>}]
 wing_b_right_01_jnt [{shapeType=Capsule, length=121.355, radius=76.7262, center=<110.8291 -4.4949 -7.5787>, rotation=<rot 0 89.2881 -70.8492>}]
 wing_b_right_02_jnt [{shapeType=Capsule, length=33.0331, radius=12.8823, center=<55.011 -2.8729 -72.7664>, rotation=<rot 0 88.7061 -73.2643>}]
 wing_c_left_01_jnt [{shapeType=Capsule, length=119.9704, radius=55.5027, center=<110.1966 -0.8099 -1.2309>, rotation=<rot 0 -89.8951 -65.4984>}]
 wing_c_left_02_jnt [{shapeType=Capsule, length=106.7764, radius=55.8096, center=<96.8733 0.6777 -63.2693>, rotation=<rot -0.1271 -90 -33.4961>}]
 wing_c_right_01_jnt [{shapeType=Capsule, length=120.2371, radius=66.0676, center=<97.6159 -1.7832 -4.9278>, rotation=<rot 0 90.4132 -69.7649>}]
 wing_c_right_02_jnt [{shapeType=Capsule, length=151.1135, radius=49.4966, center=<43.7084 0.6414 -1.7215>, rotation=<rot 0 89.9181 -45.6514>}]
 wing_d_left_01_jnt [{shapeType=Capsule, length=76.8713, radius=32.3599, center=<36.7436 -0.5096 1.9915>, rotation=<rot 0 -89.6388 -45.9644>}]
 wing_d_left_02_jnt [{shapeType=Capsule, length=142.1266, radius=47.0797, center=<54.8462 3.472 -24.0229>, rotation=<rot 0 -89.0279 -50.9631>}]
 wing_d_right_01_jnt [{shapeType=Capsule, length=95.7224, radius=35.237, center=<45.5312 0.4833 6.3731>, rotation=<rot 0 91.7945 -60.6462>}]
 wing_d_right_02_jnt [{shapeType=Capsule, length=107.8945, radius=31.7822, center=<15.3091 3.0537 -32.44>, rotation=<rot 0 90.9801 -55.7589>}]
 wing_e_left_01_jnt [{shapeType=Capsule, length=93.1334, radius=48.3998, center=<25.1351 0.9594 -27.3852>, rotation=<rot 0.7117 -90 -23.2938>}]
 wing_e_right_01_jnt [{shapeType=Capsule, length=67.4484, radius=46.2536, center=<16.8152 -0.7829 1.604>, rotation=<rot -0.8036 -90 22.0977>}]
 wrist_left_jnt [{shapeType=Capsule, length=236.1803, radius=105.3119, center=<135.287 -5.9706 74.792>, rotation=<rot 0 89.0273 -88.6817>}]
 wrist_right_jnt [{shapeType=Capsule, length=201.6683, radius=94.8115, center=<57.2133 -4.4932 -61.034>, rotation=<rot 0 88.5433 -65.0201>}]
}
constraints [
 {bone1Name=sfine_jnt, bone2Name=pelvis_JNT, swing1LimitDegrees=45, swing2LimitDegrees=45, twistLimitDegrees=45}
 {bone1Name=neck_JNT, bone2Name=sfine_jnt, swing1LimitDegrees=45, swing2LimitDegrees=45, twistLimitDegrees=45}
 {bone1Name=head_JNT, bone2Name=neck_JNT, swing1LimitDegrees=45, swing2LimitDegrees=45, twistLimitDegrees=45}
 {bone1Name=jaw_JNT, bone2Name=head_JNT, swing1LimitDegrees=45, swing2LimitDegrees=45, twistLimitDegrees=45}
 {bone1Name=hair_fin_center_a_01_jnt, bone2Name=head_JNT, swing1LimitDegrees=45, swing2LimitDegrees=45, twistLimitDegrees=45}
 {bone1Name=hair_fin_center_a_02_jnt, bone2Name=hair_fin_center_a_01_jnt, swing1LimitDegrees=45, swing2LimitDegrees=45, twistLimitDegrees=45}
 {bone1Name=hair_fin_center_b_01_jnt, bone2Name=head_JNT, swing1LimitDegrees=45, swing2LimitDegrees=45, twistLimitDegrees=45}
 {bone1Name=hair_fin_center_c_01_jnt, bone2Name=head_JNT, swing1LimitDegrees=45, swing2LimitDegrees=45, twistLimitDegrees=45}
 {bone1Name=hair_fin_center_d_01_jnt, bone2Name=head_JNT, swing1LimitDegrees=45, swing2LimitDegrees=45, twistLimitDegrees=45}
 {bone1Name=hair_fin_left_a_01_jnt, bone2Name=head_JNT, swing1LimitDegrees=45, swing2LimitDegrees=45, twistLimitDegrees=45}
 {bone1Name=hair_fin_left_a_02_jnt, bone2Name=hair_fin_left_a_01_jnt, swing1LimitDegrees=45, swing2LimitDegrees=45, twistLimitDegrees=45}
 {bone1Name=hair_fin_left_b_01_jnt2, bone2Name=head_JNT, swing1LimitDegrees=45, swing2LimitDegrees=45, twistLimitDegrees=45}
 {bone1Name=hair_fin_right_a_01_jnt, bone2Name=head_JNT, swing1LimitDegrees=45, swing2LimitDegrees=45, twistLimitDegrees=45}
 {bone1Name=hair_fin_right_b_01_jnt2, bone2Name=head_JNT, swing1LimitDegrees=45, swing2LimitDegrees=45, twistLimitDegrees=45}
 {bone1Name=shoudler_left_jnt, bone2Name=sfine_jnt, swing1LimitDegrees=45, swing2LimitDegrees=45, twistLimitDegrees=45}
 {bone1Name=elbow_left_jnt, bone2Name=shoudler_left_jnt, swing1LimitDegrees=45, swing2LimitDegrees=45, twistLimitDegrees=45}
 {bone1Name=wrist_left_jnt, bone2Name=elbow_left_jnt, swing1LimitDegrees=45, swing2LimitDegrees=45, twistLimitDegrees=45}
 {bone1Name=wing_a_left_01_jnt, bone2Name=wrist_left_jnt, swing1LimitDegrees=45, swing2LimitDegrees=45, twistLimitDegrees=45}
 {bone1Name=wing_a_left_02_jnt, bone2Name=wing_a_left_01_jnt, swing1LimitDegrees=45, swing2LimitDegrees=45, twistLimitDegrees=45}
 {bone1Name=wing_b_left_01_jnt, bone2Name=wrist_left_jnt, swing1LimitDegrees=45, swing2LimitDegrees=45, twistLimitDegrees=45}
 {bone1Name=wing_b_left_02_jnt, bone2Name=wing_b_left_01_jnt, swing1LimitDegrees=45, swing2LimitDegrees=45, twistLimitDegrees=45}
 {bone1Name=wing_c_left_01_jnt, bone2Name=wrist_left_jnt, swing1LimitDegrees=45, swing2LimitDegrees=45, twistLimitDegrees=45}
 {bone1Name=wing_c_left_02_jnt, bone2Name=wing_c_left_01_jnt, swing1LimitDegrees=45, swing2LimitDegrees=45, twistLimitDegrees=45}
 {bone1Name=wing_d_left_01_jnt, bone2Name=wrist_left_jnt, swing1LimitDegrees=45, swing2LimitDegrees=45, twistLimitDegrees=45}
 {bone1Name=wing_d_left_02_jnt, bone2Name=wing_d_left_01_jnt, swing1LimitDegrees=45, swing2LimitDegrees=45, twistLimitDegrees=45}
 {bone1Name=wing_e_left_01_jnt, bone2Name=elbow_left_jnt, swing1LimitDegrees=45, swing2LimitDegrees=45, twistLimitDegrees=45}
 {bone1Name=shoudler_right_jnt, bone2Name=sfine_jnt, swing1LimitDegrees=45, swing2LimitDegrees=45, twistLimitDegrees=45}
 {bone1Name=elbow_right_jnt, bone2Name=shoudler_right_jnt, swing1LimitDegrees=45, swing2LimitDegrees=45, twistLimitDegrees=45}
 {bone1Name=wrist_right_jnt, bone2Name=elbow_right_jnt, swing1LimitDegrees=45, swing2LimitDegrees=45, twistLimitDegrees=45}
 {bone1Name=wing_a_right_01_jnt, bone2Name=wrist_right_jnt, swing1LimitDegrees=45, swing2LimitDegrees=45, twistLimitDegrees=45}
 {bone1Name=wing_a_right_02_jnt, bone2Name=wing_a_right_01_jnt, swing1LimitDegrees=45, swing2LimitDegrees=45, twistLimitDegrees=45}
 {bone1Name=wing_b_right_01_jnt, bone2Name=wrist_right_jnt, swing1LimitDegrees=45, swing2LimitDegrees=45, twistLimitDegrees=45}
 {bone1Name=wing_b_right_02_jnt, bone2Name=wing_b_right_01_jnt, swing1LimitDegrees=45, swing2LimitDegrees=45, twistLimitDegrees=45}
 {bone1Name=wing_c_right_01_jnt, bone2Name=wrist_right_jnt, swing1LimitDegrees=45, swing2LimitDegrees=45, twistLimitDegrees=45}
 {bone1Name=wing_c_right_02_jnt, bone2Name=wing_c_right_01_jnt, swing1LimitDegrees=45, swing2LimitDegrees=45, twistLimitDegrees=45}
 {bone1Name=wing_d_right_01_jnt, bone2Name=wrist_right_jnt, swing1LimitDegrees=45, swing2LimitDegrees=45, twistLimitDegrees=45}
 {bone1Name=wing_d_right_02_jnt, bone2Name=wing_d_right_01_jnt, swing1LimitDegrees=45, swing2LimitDegrees=45, twistLimitDegrees=45}
 {bone1Name=wing_e_right_01_jnt, bone2Name=elbow_right_jnt, swing1LimitDegrees=45, swing2LimitDegrees=45, twistLimitDegrees=45}
 {bone1Name=tail_JNT, bone2Name=pelvis_JNT, swing1LimitDegrees=45, swing2LimitDegrees=45, twistLimitDegrees=45}
 {bone1Name=tail_fin_a_01_jnt, bone2Name=tail_JNT, swing1LimitDegrees=45, swing2LimitDegrees=45, twistLimitDegrees=45}
 {bone1Name=tail_fin_a_02_jnt, bone2Name=tail_fin_a_01_jnt, swing1LimitDegrees=45, swing2LimitDegrees=45, twistLimitDegrees=45}
 {bone1Name=tail_fin_a_03_jnt, bone2Name=tail_fin_a_02_jnt, swing1LimitDegrees=45, swing2LimitDegrees=45, twistLimitDegrees=45}
 {bone1Name=tail_fin_b_01_jnt, bone2Name=tail_JNT, swing1LimitDegrees=45, swing2LimitDegrees=45, twistLimitDegrees=45}
 {bone1Name=tail_fin_b_02_jnt, bone2Name=tail_fin_b_01_jnt, swing1LimitDegrees=45, swing2LimitDegrees=45, twistLimitDegrees=45}
 {bone1Name=tail_fin_b_03_jnt, bone2Name=tail_fin_b_02_jnt, swing1LimitDegrees=45, swing2LimitDegrees=45, twistLimitDegrees=45}
 {bone1Name=tail_fin_c_01_jnt, bone2Name=tail_JNT, swing1LimitDegrees=45, swing2LimitDegrees=45, twistLimitDegrees=45}
 {bone1Name=tail_fin_c_02_jnt, bone2Name=tail_fin_c_01_jnt, swing1LimitDegrees=45, swing2LimitDegrees=45, twistLimitDegrees=45}
 {bone1Name=tail_fin_c_03_jnt, bone2Name=tail_fin_c_02_jnt, swing1LimitDegrees=45, swing2LimitDegrees=45, twistLimitDegrees=45}
 {bone1Name=tail_fin_d_01_jnt, bone2Name=tail_JNT, swing1LimitDegrees=45, swing2LimitDegrees=45, twistLimitDegrees=45}
 {bone1Name=tail_fin_d_02_jnt, bone2Name=tail_fin_d_01_jnt, swing1LimitDegrees=45, swing2LimitDegrees=45, twistLimitDegrees=45}
 {bone1Name=tail_fin_d_03_jnt, bone2Name=tail_fin_d_02_jnt, swing1LimitDegrees=45, swing2LimitDegrees=45, twistLimitDegrees=45}
 {bone1Name=tail_fin_e_01_jnt, bone2Name=tail_JNT, swing1LimitDegrees=45, swing2LimitDegrees=45, twistLimitDegrees=45}
 {bone1Name=tail_fin_e_02_jnt, bone2Name=tail_fin_e_01_jnt, swing1LimitDegrees=45, swing2LimitDegrees=45, twistLimitDegrees=45}
 {bone1Name=tail_fin_f_01_jnt, bone2Name=tail_JNT, swing1LimitDegrees=45, swing2LimitDegrees=45, twistLimitDegrees=45}
 {bone1Name=tail_fin_f_02_jnt, bone2Name=tail_fin_f_01_jnt, swing1LimitDegrees=45, swing2LimitDegrees=45, twistLimitDegrees=45}
 {bone1Name=tail_fin_g_01_jnt, bone2Name=tail_JNT, swing1LimitDegrees=45, swing2LimitDegrees=45, twistLimitDegrees=45}
 {bone1Name=tail_fin_g_02_jnt, bone2Name=tail_fin_g_01_jnt, swing1LimitDegrees=45, swing2LimitDegrees=45, twistLimitDegrees=45}
 {bone1Name=tail_fin_g_03_jnt, bone2Name=tail_fin_g_02_jnt, swing1LimitDegrees=45, swing2LimitDegrees=45, twistLimitDegrees=45}
 {bone1Name=tail_fin_h_01_jnt, bone2Name=tail_JNT, swing1LimitDegrees=45, swing2LimitDegrees=45, twistLimitDegrees=45}
 {bone1Name=tail_fin_h_02_jnt, bone2Name=tail_fin_h_01_jnt, swing1LimitDegrees=45, swing2LimitDegrees=45, twistLimitDegrees=45}
 {bone1Name=tail_fin_h_03_jnt, bone2Name=tail_fin_h_02_jnt, swing1LimitDegrees=45, swing2LimitDegrees=45, twistLimitDegrees=45}
 {bone1Name=tail_fin_i_01_jnt, bone2Name=tail_JNT, swing1LimitDegrees=45, swing2LimitDegrees=45, twistLimitDegrees=45}
 {bone1Name=tail_fin_i_02_jnt, bone2Name=tail_fin_i_01_jnt, swing1LimitDegrees=45, swing2LimitDegrees=45, twistLimitDegrees=45}
 {bone1Name=tail_fin_i_03_jnt, bone2Name=tail_fin_i_02_jnt, swing1LimitDegrees=45, swing2LimitDegrees=45, twistLimitDegrees=45}
 {bone1Name=thigh_left_jnt, bone2Name=pelvis_JNT, swing1LimitDegrees=45, swing2LimitDegrees=45, twistLimitDegrees=45}
 {bone1Name=tarsus_left_jnt, bone2Name=thigh_left_jnt, swing1LimitDegrees=45, swing2LimitDegrees=45, twistLimitDegrees=45}
 {bone1Name=foot_left_jnt, bone2Name=tarsus_left_jnt, swing1LimitDegrees=45, swing2LimitDegrees=45, twistLimitDegrees=45}
 {bone1Name=toe_left_jnt, bone2Name=foot_left_jnt, swing1LimitDegrees=45, swing2LimitDegrees=45, twistLimitDegrees=45}
 {bone1Name=claw_left_jnt, bone2Name=toe_left_jnt, swing1LimitDegrees=45, swing2LimitDegrees=45, twistLimitDegrees=45}
 {bone1Name=hide_toe_left_jnt, bone2Name=foot_left_jnt, swing1LimitDegrees=45, swing2LimitDegrees=45, twistLimitDegrees=45}
 {bone1Name=thigh_right_jnt, bone2Name=pelvis_JNT, swing1LimitDegrees=45, swing2LimitDegrees=45, twistLimitDegrees=45}
 {bone1Name=tarsus_right_jnt, bone2Name=thigh_right_jnt, swing1LimitDegrees=45, swing2LimitDegrees=45, twistLimitDegrees=45}
 {bone1Name=foot_right_jnt, bone2Name=tarsus_right_jnt, swing1LimitDegrees=45, swing2LimitDegrees=45, twistLimitDegrees=45}
 {bone1Name=toe_right_jnt, bone2Name=foot_right_jnt, swing1LimitDegrees=45, swing2LimitDegrees=45, twistLimitDegrees=45}
 {bone1Name=claw_right_jnt, bone2Name=toe_right_jnt, swing1LimitDegrees=45, swing2LimitDegrees=45, twistLimitDegrees=45}
 {bone1Name=hide_toe_right_jnt, bone2Name=foot_right_jnt, swing1LimitDegrees=45, swing2LimitDegrees=45, twistLimitDegrees=45}
]
